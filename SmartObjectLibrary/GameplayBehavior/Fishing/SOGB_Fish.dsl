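; Smart Object gameplay behavior that makes an AI bot fish.
; Init validates config, pawn and bot controller, then the bot receives and equips a rod, ending on failure.
; Chained delay tasks pulse fire to cast, lock equipment changes, push behavior control and hold for the configured fishing delay.
; A Threatened or Suspicious alert ends it early, and finishing restores control and abilities, removes the rod and unbinds the alert.
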(event GameplayBehavior|EventOnTriggeredCharacter
  (params (Avatar :type Character%20Object%20Reference) (Config :type Gameplay%20Behavior%20Config%20Object%20Reference) (SmartObjectOwner :type Actor%20Object%20Reference))
  ;[1 1 1] On Start Behavior
  (bind _isvalid (CallFunction|Init Avatar Config SmartObjectOwner))
  (if _isvalid
    (CallFunction|HandleStartFishing)
    ; When attacked, cancel the behavior
    (AlertLevel|BindEventtoOnAlertLevelChangedEvent (Variables|InternalData|GetUserController) (delegate Custom|EndBehaviorWhenThreatened))
    (else
      (GameplayBehavior|EndBehavior Avatar))))

(event Custom|EndBehaviorWhenThreatened
  (params (BotController :type Fort%20Athena%20AIBot%20Controller%20Object%20Reference) (OldAlertLevel :type EAlertLevel%20Enum :default "Unaware") (NewAlertLevel :type EAlertLevel%20Enum :default "Unaware"))
  ; When attacked, cancel the behavior
  (if (or (Utilities|Enum|Equal(Enum) NewAlertLevel "Threatened") (Utilities|Enum|Equal(Enum) NewAlertLevel "Suspicious"))
    (GameplayBehavior|EndBehavior (Variables|Default|GetUserAvatar))))

(event Custom|HandleStartFishing
  ; Handle Start Fishing
  (CallFunction|TryGiveFishingRod)
  (bind _returnvalue (Class|FortAthenaAIBotController|EquipWeaponbyTag (Variables|InternalData|GetUserController) (Variables|Default|GetTag_FishingRod)))
  (Utilities|IsValid _returnvalue
    (:Is%20Valid
      (GameplayTasks|TaskWaitDelay 0 1.0
        (:OnFinish
          (Class|FortAthenaAIBotController|StartFire (Variables|InternalData|GetUserController))
          (GameplayTasks|TaskWaitDelay 0 1.25
            (:OnFinish
              (Class|FortAthenaAIBotController|StopFire (Variables|InternalData|GetUserController))
              (Tags|BlockAbilitiesWithTags (Utilities|Casting|CastToFortAbilitySystemComponent (Ability|GetAbilitySystemComponent (Variables|InternalData|GetUserPawn))) (Variables|Default|GetTAG_ChangeEquipment))
              (AI|PushBehaviorControl (Variables|InternalData|GetUserController) -4185085)
              (GameplayTasks|TaskWaitDelay 0 (Variables|/SmartObjectLibrary/GameplayBehavior/Fishing/SOGB_Fishing_Config.SOGB_Fishing_Config_C|GetFishingDelay (Variables|Default|GetFishingConfig))
                (:OnFinish
                  (Class|FortAthenaAIBotController|StopFire (Variables|InternalData|GetUserController))
                  (GameplayTasks|TaskWaitDelay 0 1.25
                    (:OnFinish
                      (GameplayBehavior|EndBehavior (Variables|Default|GetUserAvatar)))))))))))
    (:Is%20Not%20Valid
      (Fortnite|FortShippingLog self "AI could not Equip Fishing Rod by tag! Validate that the AI's Behavior Component does not Support Weapon Holstering - If it allows Weapon Holstering, for now the AI will not be able to equip the Fishing Rod." true)
      (GameplayBehavior|EndBehavior (Variables|Default|GetUserAvatar)))))

(event Custom|TryGiveFishingRod
  (bind _usercontroller (Variables|InternalData|GetUserController))
  ; Try to give the AI the Fishing Rod Item
  (Class|FortAthenaAIBotController|SetIgnoreInventoryRestrictions _usercontroller true)
  (bind _returnvalue (Inventory|GiveItemtoInventoryOwner _usercontroller "/Game/Athena/Items/Consumables/FloppingRabbit/WID_Athena_FloppingRabbit.WID_Athena_FloppingRabbit"))
  (bind _output_get (Variables|InternalData|SetFishingRod _returnvalue))
  (Utilities|IsValid _output_get
    (:Is%20Not%20Valid
      (Fortnite|FortShippingLog self "AI could not receive the Fishing Rod in his Inventory! This is probably because of Loot Restrictions! Make sure in the Inventory Component that the AI can ignore restrictions to be able to receive the Fishing Rod in his inventory!" true)
      (GameplayBehavior|EndBehavior (Variables|Default|GetUserAvatar)))))

(event GameplayBehavior|EventOnFinishedCharacter
  (params (Avatar :type Character%20Object%20Reference) (bWasInterrupted :type Boolean))
  (bind _usercontroller (Variables|InternalData|GetUserController))
  (bind _userpawn (Variables|InternalData|GetUserPawn))
  ;[1 1 1] On Stop Behavior
  (Utilities|IsValid _usercontroller
    (:Is%20Valid
      (AI|PopBehaviorControl _usercontroller)
      (Utilities|IsValid _userpawn
        (:Is%20Valid
          (Tags|UnBlockAbilitiesWithTags (Utilities|Casting|CastToFortAbilitySystemComponent (Ability|GetAbilitySystemComponent _userpawn)) (Variables|Default|GetTAG_ChangeEquipment) true)
          (InventoryManagement|RemoveItem (Variables|InternalData|GetUserController) (Item|GetItemGuid (Variables|InternalData|GetFishingRod)) 1 true)
          (AlertLevel|UnbindEventfromOnAlertLevelChangedEvent (Variables|InternalData|GetUserController) (EventDispatchers|CreateEvent :!event Custom|EndBehaviorWhenThreatened)))))))

(fn Init
  (params (Avatar :type Character%20Object%20Reference) (Config :type Gameplay%20Behavior%20Config%20Object%20Reference) (SmartObjectOwner :type Actor%20Object%20Reference))
  (Variables|Default|SetUserAvatar Avatar)
  (Variables|Default|SetSmartObjectOwner SmartObjectOwner)
  (bind _assogb_fishing_config (Utilities|Casting|CastToSOGB_Fishing_Config Config)
    (:then
      (bind _output_get (Variables|Default|SetFishingConfig _assogb_fishing_config))
      (if (> (Variables|/SmartObjectLibrary/GameplayBehavior/Fishing/SOGB_Fishing_Config.SOGB_Fishing_Config_C|GetFishingDelay _output_get) 0)
        (bind _asfort_player_pawn (Utilities|Casting|CastToFortPlayerPawn (Variables|GetAvatar))
          (:then
            (bind _output_get_2 (Variables|InternalData|SetUserPawn _asfort_player_pawn))
            (bind _asfort_athena_aibot_controller (Utilities|Casting|CastToFortAthenaAIBotController (Pawn|GetController _output_get_2))
              (:then
                (Variables|InternalData|SetUserController _asfort_athena_aibot_controller)
                (return true))
              (:CastFailed
                _)))
          (:CastFailed
            _))
        (return false)
        (else
          _)))
    (:CastFailed
      _)))
